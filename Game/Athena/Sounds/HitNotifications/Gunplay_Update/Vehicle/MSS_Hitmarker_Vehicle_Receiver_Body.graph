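# Stereo one-shot MetaSound hitmarker, played by its name for the receiver of a vehicle body hit.
# After a short trigger delay it randomly picks one of three body-hit waves, with randomized pitch and decibel gain.
# A receiver volume modulator scales the gain, then hit azimuth pans the result; other HitMarker inputs are unused.
# Sends to sidechain buses, including weapon, explosion and instigator hitmarker ones, accompany receiver concurrency settings.
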MetaSoundSource /Game/Athena/Sounds/HitNotifications/Gunplay_Update/Vehicle/MSS_Hitmarker_Vehicle_Receiver_Body
inputs [
 HitMarker.Azimuth {type=Float}
 HitMarker.DamageDealt {type=Float}
 HitMarker.Distance {type=Float}
 HitMarker.InstigatorAccumulatedDamage {type=Float}
 HitMarker.ReceiverPercentHealth {type=Float}
 HitMarker.ReceiverPercentShield {type=Float}
 Random.Pitch.Base {type=Float}
 Random.Pitch.Spread {type=Float}
 Random.Volume.Base {type=Float}
 Random.Volume.Spread {type=Float}
 Sounds {type=WaveAsset:Array}
 UE.Source.OnPlay {type=Trigger}
]
interfaces [Fort.HitMarker:1.0, UE.OutputFormat.Stereo:1.0, UE.Source.OneShot:1.0, UE.Source:1.1]
outputs [
 "UE.OutputFormat.Stereo.Audio:0" {type=Audio}
 "UE.OutputFormat.Stereo.Audio:1" {type=Audio}
 UE.Source.OneShot.OnFinished {type=Trigger}
]
pages [Default {
  input_defaults {
   HitMarker.Azimuth=0
   HitMarker.DamageDealt=0
   HitMarker.Distance=0
   HitMarker.InstigatorAccumulatedDamage=0
   HitMarker.ReceiverPercentHealth=1
   HitMarker.ReceiverPercentShield=1
   Random.Pitch.Base=0.492
   Random.Pitch.Spread=0.532
   Random.Volume.Base=0
   Random.Volume.Spread=0
   UE.Source.OnPlay=null
   Sounds [
    &/Hitmarker_Vehicle_Receiver_Body_01
    &/Hitmarker_Vehicle_Receiver_Body_02
    &/Hitmarker_Vehicle_Receiver_Body_03
   ]
  }
  nodes [
   n1 = AbsoluteValue.Abs:Float(Input=^Random.Volume.Spread)
   DecibelstoLinearGain_0 = UE.Decibels to Linear Gain:Float(Decibels=UE.RandomFloat(
     Next=^UE.Source.OnPlay
     Max=UE.Add:Float(PrimaryOperand=^Random.Volume.Base, AdditionalOperands=@n1.Value).Out
     Min=UE.Subtract:Float(PrimaryOperand=^Random.Volume.Base, AdditionalOperands=@n1.Value).Out
    ).Value)
   n2 = UE.Multiply:Float(
    PrimaryOperand=@DecibelstoLinearGain_0.Linear Gain
    AdditionalOperands=Modulation.GetModulatorValue() {also {defaults {Modulator=/Game/Athena/Sounds/HitNotifications/Gunplay_Update/Modulation/Patch_Hitmarkers_Receiver_Standard_Volume}}}.Out
   )
   TriggerDelay_0 = UE.Trigger Delay(In=^UE.Source.OnPlay) {also {defaults {Delay Time=0.1}}}
   ArrayRandomGet_0 = Array.Random Get:WaveAsset:Array(In Array=^Sounds, Next=@TriggerDelay_0.Out) {also {defaults {No Repeats=-1}}}
   n3 = AbsoluteValue.Abs:Float(Input=^Random.Pitch.Spread)
   WavePlayer_0 = UE.Wave Player:Stereo(
    Play=@ArrayRandomGet_0.On Next
    Wave Asset=@ArrayRandomGet_0.Value
    Pitch Shift=UE.RandomFloat(
     Next=^UE.Source.OnPlay
     Max=UE.Add:Float(PrimaryOperand=^Random.Pitch.Base, AdditionalOperands=@n3.Value).Out
     Min=UE.Subtract:Float(PrimaryOperand=^Random.Pitch.Base, AdditionalOperands=@n3.Value).Out
    ).Value
   ) {also {version [1,1]}}
   AzimuthPannerStereo_0 = &AzimuthPannerStereo(
    Azimuth=^HitMarker.Azimuth
    In L=UE.Multiply:Audio by Float(AdditionalOperands=@n2.Out, PrimaryOperand=@WavePlayer_0.Out Left).Out
    In R=UE.Multiply:Audio by Float(AdditionalOperands=@n2.Out, PrimaryOperand=@WavePlayer_0.Out Right).Out
   )
  ]
  links [
   {dst="@UE.OutputFormat.Stereo.Audio:0", src=AzimuthPannerStereo_0.Out L}
   {dst="@UE.OutputFormat.Stereo.Audio:1", src=AzimuthPannerStereo_0.Out R}
   {dst="@UE.Source.OneShot.OnFinished", src=WavePlayer_0.On Finished}
  ]
 }]
properties {
 assetImportData=&self:AssetImportData
 audioPropertiesSheet=/Game/Athena/Sounds/HitNotifications/Gunplay_Update/APS_Hitmarker_Receiver_Standard
 numChannels=2
 outputFormat=Stereo
 sampleRate=48000
 soundClassObject=&Fort_Hitmarkers
 soundSubmixObject=&Fortnite_SFX
 volume=1.6
 busSends [
  {audioBus=&AB_Sidechain_FortWeapon_Vol, sendLevel=0.3}
  {audioBus=&AB_Sidechain_FortExplosion_Vol, sendLevel=0.4}
  {audioBus=&AB_Sidechain_FortAmbient_Vol, sendLevel=0.3}
  {audioBus=&AB_Sidechain_FortLoot_Vol, sendLevel=0.3}
  {audioBus=&AB_Sidechain_FortVehicles_Vol, sendLevel=0.3}
  {audioBus=&AB_Sidechain_FortCreatures_Vol, sendLevel=0.3}
  {audioBus=&AB_Sidechain_FortEmote_Vol, sendLevel=0.3}
  {audioBus=/Game/Athena/Sounds/HitNotifications/Gunplay_Update/Modulation/AB_Sidechain_Hitmarkers_Instigator_Critical_Bus, sendLevel=0.7}
  {audioBus=/Game/Athena/Sounds/HitNotifications/Gunplay_Update/Modulation/AB_Sidechain_Hitmarkers_Instigator_Standard_Bus, sendLevel=0.7}
 ]
 concurrencySet [
  /Game/Athena/Sounds/HitNotifications/Gunplay_Update/CON_Hitmarker_GunplayUpdate_Receiver_Standard
  /Game/Athena/Sounds/HitNotifications/Gunplay_Update/Settings/CONC_Hitmarker_GunplayUpdate_Receiver_Health
 ]
 modulationSettings {
  volumeRouting=Union
  volumeModulationDestination {bEnableModulation=true, modulators [&Patch_FortHitNotify_Vol]}
 }
}
